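# Material function in the TankFlare wheel cosmetic folder that outputs an object's world pivot, scale and a world-space vector.
# It transforms constant points to world space: the origin gives the pivot, and (1,1,1) minus that pivot gives XYZ scale, which rotation affects.
# Uniform scale measures 0.5774-diagonal to (1,1,1), not to the pivot, yielding about 0.73 times true scale.
# A Transform node converts the Local Vector input to World Vector.
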MaterialFunction /VehicleCosmetics/Wheels/TankFlare/Materials/ObjectLocaltoWorld
graph {nodes [
  n1 = TransformPosition(None=Constant4Vector() {constant=<rgba 0.5774 0.5774 0.5774 1>}.RGBA) {transformType=TRANSFORMPOSSOURCE_World}
  n2 = ComponentMask(None=TransformPosition(None=Constant4Vector() {constant=<rgba 1 1 1 1>}.RGBA) {transformType=TRANSFORMPOSSOURCE_World}) {b=true, g=true, r=true}
  FunctionOutput(None=Distance(A=ComponentMask(None=@n1) {b=true, g=true, r=true}, B=@n2)) {description=Returns a scalar value that is equal to the objects uniform scale., outputName=Object Scale, sortPriority=2}
  n3 = ComponentMask(None=TransformPosition(None=Constant4Vector() {constant=<rgba 0 0 0 1>}.RGBA) {transformType=TRANSFORMPOSSOURCE_World}) {b=true, g=true, r=true}
  FunctionOutput(None=Subtract(A=@n2, B=@n3)) {description="Returns a vector value that is equal to the objects scale on the x,y and z axes.", outputName=Object Scale XYZ, sortPriority=2}
  n4 = FunctionInput() {
   bUsePreviewValueAsDefault=true
   description="Input a vector 3 value to be transformed by the objects rotation. For example, a value of (1,0,0) will return a world space vector along the objects x-axis regardless of object orientation."
   inputName=Local Vector
   previewValue=<0 0 1 1>
  }
  FunctionOutput(None=Transform(None=@n4)) {description=Local vector input transformed into world space., outputName=World Vector, sortPriority=1}
  FunctionOutput(None=@n3) {description=Object pivot point in world space., outputName=Object Pivot Location}
 ]}
