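# Material function that decodes Pivot Painter data baked into a mesh, kept with the TankFlare wheel cosmetic's materials.
# Pivot position comes from UV channel 2 plus UV 3's first component, transformed to world space.
# Vertex color RGB expands to a signed, normalized child X-axis, transformed from local space.
# It also outputs a per-element random value from UV 3, vertex-alpha custom falloff and the object's uniform scale.
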MaterialFunction /VehicleCosmetics/Wheels/TankFlare/Materials/Pivot_Painter_PerPointData
graph {nodes [
  n1 = TextureCoordinate() {coordinateIndex=3}
  n2 = AppendVector(A=AppendVector(A=TextureCoordinate() {coordinateIndex=2}, B=ComponentMask(None=@n1) {r=true}), B=Constant() {r=1})
  FunctionOutput(None=ComponentMask(None=TransformPosition(None=@n2) {transformType=TRANSFORMPOSSOURCE_World}) {b=true, g=true, r=true}) {outputName=Pivot Position}
  n3 = Normalize(VectorInput=ConstantBiasScale(None=VertexColor()) {bias=-0.5, scale=2})
  FunctionOutput(None=Normalize(VectorInput=Transform(None=@n3) {transformSourceType=TRANSFORMSOURCE_Local})) {
   description=Returns a normalized vector pointing down the elements X-Axis. Which points from the pivot point toward the averaged center of the mesh.
   outputName=Child X-Axis
   sortPriority=12
  }
  FunctionOutput(None=ComponentMask(None=TextureCoordinate() {coordinateIndex=3}) {g=true}) {description=Returns a random value per element in the 0-1 range., outputName=Random Value Per Element, sortPriority=13}
  FunctionOutput(None=ComponentMask(None=VertexColor().A) {r=true}) {
   description=Returns the custom falloff values stored in the vertex alpha channel of the model.
   outputName=Custom Alpha Values
   sortPriority=14
  }
  FunctionOutput(None=MaterialFunctionCall!(Local Vector) {materialFunction=&/ObjectLocaltoWorld}.Object Scale) {description=Returns a scalar value that is equal to the objects uniform scale., outputName=Object Scale, sortPriority=20}
 ]}
