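# Parent material for TankFlare vehicle wheels, with animated emissive glow and optional vertex jitter.
# Glow combines mask channels, noise-distorted panning clouds, Fresnel, vertex color and a FlashSpeed sine-cosine pulse, tinted between Color1 and CustomColor.
# The Painted switch keeps only the masked flash in CustomColor.
# Pivot Painter data drives World Position Offset jitter, zero by default, and attributes pass through vehicle damage and de-rez transition functions.
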Material /VehicleCosmetics/Wheels/TankFlare/Materials/M_WheelParent_TankFlare_01
graph {
 properties {bAutomaticallySetUsageInEditor=false, bUseMaterialAttributes=true, bUsedWithSkeletalMesh=true, usedShadingModels=Default Lit}
 nodes [
  Divide()
  n1 = VectorParameter() {parameterName=CustomColor, defaultValue=<rgba 1 0.0369 0 1>}
  n2 = Multiply(
   A=ConstantBiasScale(None=VectorParameter() {parameterName=Color1, defaultValue=<rgba 0.1701 0.1669 0.1701 1>}.RGB) {bias=0, scale=2}
   B=Constant3Vector() {constant=<rgba 0.9918 1 0.9448 0>}
  )
  n3 = VectorParameter() {parameterName=Color1, defaultValue=<rgba 0.1701 0.1669 0.1701 1>}
  n4 = NamedRerouteUsage()
  n5 = Clamp(None=ConstantBiasScale(None=NamedRerouteUsage()) {bias=0, scale=10})
  n6 = TextureSampleParameter2D() {
   parameterName=Mask_Emissive
   samplerType=SAMPLERTYPE_LinearColor
   texture=/VehicleCosmetics/Wheels/TankFlare/Textures/T_Tank_Flare_Rim_MASK_MAT
  }
  n7 = Max(A=@n6.R, B=@n6.R)
  n8 = Multiply(A=Time(), B=ScalarParameter() {defaultValue=1, parameterName=FlashSpeed})
  n9 = Sine(None=@n8)
  n10 = Constant() {r=0.5}
  n11 = Multiply(A=@n9, B=Cosine(None=Multiply(A=@n8, B=@n10)))
  n12 = Multiply(A=Max(A=@n7, B=NamedRerouteUsage()), B=@n11)
  n13 = StaticSwitchParameter(True=@n1.RGB, False=LinearInterpolate(A=LinearInterpolate(A=@n2, B=@n3.RGB, Alpha=@n4), B=@n1.RGB, Alpha=Max(A=@n5, B=@n12))) {parameterName=Painted}
  n14 = Multiply(A=Multiply(A=@n7, B=@n11), B=Constant() {r=0.1})
  n15 = Constant() {r=0.1}
  n16 = TextureCoordinate() {uTiling=10, vTiling=7.5}
  n17 = TextureSample(UVs=Panner(Coordinate=@n16) {speedY=0.5}) {texture=/VehicleCosmetics/Wheels/TankFlare/Textures/T_Tank_Flare_Noise_RGB}
  n18 = Panner(Coordinate=Add(A=Multiply(A=@n15, B=@n17.B), B=@n16)) {speedX=0.5, speedY=-0.2}
  n19 = Subtract(
   A=Clamp(None=TextureSample(UVs=@n18) {texture=/VehicleCosmetics/Wheels/_SharedTextures/FX/Clouds_RGB_Soft_03_T}.R)
   B=Clamp(None=ConstantBiasScale(None=@n17.B) {bias=-0.2, scale=1})
  )
  n20 = Max(A=@n19, B=@n19)
  n21 = Time()
  n22 = Abs(None=Cosine(None=ConstantBiasScale(None=@n21) {bias=0, scale=0.2}))
  n23 = Multiply(A=Add(A=Clamp(None=@n20), B=Constant() {r=0.1}), B=ConstantBiasScale(None=Clamp(None=@n22)) {bias=0.75, scale=2})
  n24 = Multiply(A=@n6.G, B=Constant() {r=3})
  n25 = ConstantBiasScale(None=@n6.R) {bias=0, scale=0.05}
  n26 = ConstantBiasScale(None=@n6.B) {bias=0, scale=0.05}
  n27 = MaterialFunctionCall!(
   Normal=TextureSampleParameter2D() {parameterName=Normal, samplerType=SAMPLERTYPE_Normal, texture=/VehicleCosmetics/Wheels/TankFlare/Textures/T_Tank_Flare_Rim_N}.RGB
   Exponent=ScalarParameter() {defaultValue=1, parameterName=Fresnel_Exp}
  ) {materialFunction=/VehicleCosmetics/Wheels/Cola_Holographic/Materials/MF_Fresnel_Classic}
  n28 = Add(A=Multiply(A=Abs(None=@n11), B=Constant() {r=-0.2}), B=ScalarParameter() {defaultValue=1, parameterName=Fresnel_Light})
  n29 = VertexColor()
  n30 = ConstantBiasScale(None=OneMinus(None=@n29.R)) {bias=0.05, scale=0.075}
  n31 = Add(A=Add(A=Multiply(A=@n23, B=@n24), B=Subtract(A=@n25, B=@n26)), B=Add(A=Power(Base=@n27.Result, Exp=@n28), B=Clamp(None=@n30)))
  n32 = Multiply(A=StaticSwitchParameter(True=@n14, False=Add(A=@n14, B=@n31)) {parameterName=Painted})
  n33 = SetMaterialAttributes!(
   MaterialAttributes
   Metallic=ScalarParameter() {parameterName=Metallic}
   Roughness=ScalarParameter() {defaultValue=1, parameterName=Roughness}
   Emissive Color=Multiply(A=Multiply(A=@n32, B=@n13), B=ScalarParameter() {defaultValue=2, parameterName=EmissiveGlowIntensity})
   World Position Offset=NamedRerouteUsage()
   Normal=TextureSampleParameter2D() {parameterName=Normal, samplerType=SAMPLERTYPE_Normal, texture=&T_Tank_Flare_Rim_N}.RGB
  ) {attributeSetTypes [
    #1
    #2
    #3
    #4
    #5
   ]}
  MaterialFunctionCall(In=MaterialFunctionCall(Material=@n33) {materialFunction=&MF_BaseVehicle_Damage}.Result) {materialFunction=&MF_VehicleDeRez_Transition}
  Max(A=Clamp(None=Constant() {r=1}), B=@n11)
  n34 = OneMinus(None=NamedRerouteUsage())
  n35 = Constant() {r=0.5}
  n36 = Multiply(A=Cosine(None=Multiply(A=@n8, B=@n35)), B=@n9)
  n37 = MaterialFunctionCall() {materialFunction=&/Pivot_Painter_PerPointData}
  n38 = Multiply(A=Multiply(A=Multiply(A=@n34, B=@n36), B=ScalarParameter() {defaultValue=0, parameterName=Jitter}), B=@n37.Child X-Axis)
  n39 = AppendVector(A=ComponentMask(None=@n38) {g=true}, B=Constant())
  n40 = ComponentMask(None=@n38) {g=true}
  n41 = Multiply(A=@n36, B=Multiply(A=Time(), B=Constant() {r=1}))
  n42 = ConstantBiasScale(None=Sine(None=Add(A=@n41, B=@n37.Random Value Per Element))) {bias=0.4, scale=1}
  n43 = OneMinus(None=NamedRerouteUsage())
  NamedRerouteDeclaration(None=LinearInterpolate(A=Constant(), B=AppendVector(A=@n39, B=@n40), Alpha=Multiply(A=@n42, B=@n43))) {name=WPO, nodeColor=<rgba 0 1 0.3636 1>}
  NamedRerouteDeclaration(None=@n29.R) {name=R_VC, nodeColor=<rgba 0 1 0.5312 1>}
  NamedRerouteDeclaration(None=@n6.G) {name=Green_CM, nodeColor=<rgba 1 0.2109 0 1>}
 ]
}
