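# Masked material for the TankFlare vehicle wheel decal, glowing in a tintable paint color.
# The glow takes the greater of the car's speed ratio and an irregular flicker built from multiplied sine and cosine waves, masked by the opacity texture's red channel.
# A normal-mapped Fresnel is added before tinting into base color and emissive, with the green channel as opacity mask.
# Shared vehicle damage and de-rez transition functions wrap the result.
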Material /VehicleCosmetics/Wheels/TankFlare/Materials/M_WheelParent_TankFlare_Dekal
graph {
 properties {
  bAutomaticallySetUsageInEditor=false
  bUseMaterialAttributes=true
  bUsedWithMeshDeformer=true
  bUsedWithSkeletalMesh=true
  blendMode=BLEND_Masked
  previewMesh=/VehicleCosmetics/Wheels/Nico1Radiant/Meshes/SK_Wheel_Nico1Radiant
  usedShadingModels=Default Lit
 }
 nodes [
  n1 = TextureSample() {samplerType=SAMPLERTYPE_Normal, texture=/VehicleCosmetics/Wheels/TankFlare/Textures/T_Tank_Flare_Decal_N}
  n2 = Constant() {r=10}
  n3 = Divide(
   A=ScalarParameter() {parameterName=Game_OBJ_Param:_Car_Object_Speed}
   B=ScalarParameter() {defaultValue=1, parameterName=Game_OBJ_Param:_Car_Max_Speed}
  )
  n4 = Multiply(A=Time(), B=ScalarParameter() {defaultValue=1, parameterName=FlashSpeed})
  n5 = Multiply(A=Constant() {r=0.5}, B=@n4)
  n6 = Multiply(A=Constant() {r=0.35}, B=@n4)
  n7 = Multiply(A=Sine(None=@n4), B=Multiply(A=Cosine(None=@n5), B=Cosine(None=@n6)))
  n8 = TextureSample() {samplerType=SAMPLERTYPE_LinearColor, texture=/VehicleCosmetics/Wheels/TankFlare/Textures/T_Tank_Flare_Decal_Opacity}
  n9 = Multiply(A=Max(A=Clamp(None=@n3), B=Abs(None=@n7)), B=@n8.R)
  n10 = Multiply(
   A=Add(
    A=MaterialFunctionCall!(Normal=@n1.RGB, Exponent) {materialFunction=/VehicleCosmetics/Wheels/Cola_Holographic/Materials/MF_Fresnel_Classic}.Result
    B=Multiply(A=@n2, B=@n9)
   )
   B=VectorParameter() {parameterName=PaintColor, defaultValue=<rgba 1 0.03 0 1>}.RGB
  )
  MaterialFunctionCall(In=MaterialFunctionCall(Material=SetMaterialAttributes!(MaterialAttributes, Emissive Color=@n10, Opacity Mask=@n8.G, Base Color=@n10) {attributeSetTypes [#1, #2, #3]}) {materialFunction=&MF_BaseVehicle_Damage}.Result) {materialFunction=&MF_VehicleDeRez_Transition}
 ]
}
