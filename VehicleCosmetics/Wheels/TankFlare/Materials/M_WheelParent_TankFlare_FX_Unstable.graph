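# Translucent unlit parent material for the unstable TankFlare wheel effect on skeletal meshes, glowing in an emissive paint color.
# Two panning caustic textures, masked by a shockwave curve texture, form the opacity.
# An optional flash modulates it by the larger of car speed ratio or a sine pulse, optionally randomized.
# Pivot Painter per-element axes and random values drive world-position jitter, before vehicle damage and de-rez functions apply.
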Material /VehicleCosmetics/Wheels/TankFlare/Materials/M_WheelParent_TankFlare_FX_Unstable
graph {
 properties {
  bAutomaticallySetUsageInEditor=false
  bUseMaterialAttributes=true
  bUsedWithSkeletalMesh=true
  blendMode=BLEND_Translucent
  shadingModel=MSM_Unlit
  usedShadingModels=Unlit
 }
 nodes [
  n1 = Multiply(
   A=VectorParameter() {parameterName=PaintColor, defaultValue=<rgba 0.19 0.84 0.67 1>}.RGB
   B=ScalarParameter() {defaultValue=32, parameterName=EmissiveMultiply}
  )
  n2 = TextureCoordinate()
  n3 = AppendVector(
   A=Multiply(A=ComponentMask(None=@n2) {r=true}, B=ScalarParameter() {defaultValue=1, parameterName=TileX})
   B=Multiply(A=ComponentMask(None=@n2) {g=true}, B=ScalarParameter() {defaultValue=1, parameterName=TileY})
  )
  n4 = Multiply(A=Time(), B=ScalarParameter() {defaultValue=1, parameterName=PanSpeedAdjust})
  n5 = TextureSampleParameter2D(UVs=Panner(Coordinate=@n3, Time=@n4) {speedX=0.0754, speedY=0.0675}) {parameterName=Caustic_MAP1, texture=/VehicleCosmetics/Wheels/Secretooze/Textures/T_Caustics_D}
  n6 = Multiply(A=@n5.G, B=@n5.G)
  n7 = TextureSampleParameter2D(UVs=Panner(Coordinate=@n3, Time=@n4) {speedX=0.125, speedY=1}) {parameterName=Caustic_MAP2, texture=&T_Caustics_D}
  n8 = TextureSample() {texture=/VehicleCosmetics/Wheels/_SharedTextures/FX/Curve_Shockwaves_Pack}
  n9 = Add(A=Multiply(A=Add(A=@n7.R, B=@n6), B=@n8.R), B=Multiply(A=@n8.R, B=Constant() {r=0.025}))
  n10 = Clamp(None=Divide(A=ScalarParameter() {parameterName=GOPST_ObjectSpeed}, B=ScalarParameter() {parameterName=GOPST_CarMaxSpeed}))
  n11 = Multiply(A=Time(), B=ScalarParameter() {defaultValue=0.5, parameterName=FlashSpeed})
  n12 = Sine(None=@n11)
  n13 = Cosine(None=Multiply(A=@n11, B=Constant() {r=0.5}))
  n14 = Constant() {r=0.35}
  n15 = Multiply(A=@n13, B=Cosine(None=Multiply(A=@n11, B=@n14)))
  n16 = Constant() {r=1}
  n17 = Abs(None=Multiply(A=@n12, B=StaticSwitchParameter(True=@n15, False=@n16) {parameterName=Flash_Randomizer}))
  n18 = StaticSwitchParameter(True=Multiply(A=@n9, B=Max(A=@n10, B=@n17)), False=@n9) {parameterName=Flash}
  n19 = Multiply(A=@n12, B=@n13)
  n20 = OneMinus(None=Constant() {r=-1})
  n21 = MaterialFunctionCall() {materialFunction=&/Pivot_Painter_PerPointData}
  n22 = Multiply(A=Multiply(A=Multiply(A=@n19, B=@n20), B=ScalarParameter() {defaultValue=0.0003, parameterName=Jitter}), B=@n21.Child X-Axis)
  n23 = ComponentMask(None=@n22) {g=true}
  n24 = Constant()
  n25 = Multiply(A=Time(), B=Constant() {r=1})
  n26 = Sine(None=Add(A=Multiply(A=@n25, B=@n19), B=@n21.Random Value Per Element))
  n27 = LinearInterpolate(
   A=AppendVector(A=AppendVector(A=@n23, B=@n24), B=ComponentMask(None=@n22) {r=true})
   B=Constant()
   Alpha=Multiply(A=@n20, B=ConstantBiasScale(None=@n26) {bias=0.4, scale=1})
  )
  MaterialFunctionCall(In=MaterialFunctionCall(Material=SetMaterialAttributes!(MaterialAttributes, Emissive Color=@n1, Opacity=@n18, World Position Offset=@n27) {attributeSetTypes [#1, #2, #3]}) {materialFunction=&MF_BaseVehicle_Damage}.Result) {materialFunction=&MF_VehicleDeRez_Transition}
  Multiply(A=Multiply(A=@n5.R, B=@n8.R), B=Constant() {r=32})
 ]
}
